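# Translucent, unlit, two-sided mesh-particle material for a soft, animated smoky cone effect in the MaskMaker takeover FX.
# Four panning inky-smoke samples blend into noise that modulates a two-color emissive, whose sphere-mask blend depends only on an offset parameter, not world position.
# Opacity combines that noise with particle alpha, a dynamic-parameter-driven UV soft mask, axis-aligned Fresnel and EventIntensity.
# Camera and depth fades are skipped on ES feature levels.
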Material /BRMapCh6/FX/MaskMakerTakeOver/Materials/M_MaskMaker_Cone_Soft
graph {
 properties {
  bUsedWithMeshParticles=true
  blendMode=BLEND_Translucent
  pixelDepthOffsetMode=PDOM_Legacy
  shadingModel=MSM_Unlit
  twoSided=true
  usedShadingModels=Unlit
 }
 nodes [
  n1 = WorldPosition()
  n2 = VectorParameter() {parameterName=SphereMaskPosition, defaultValue=<rgba 0 0 0 1>}
  n3 = LinearInterpolate(
   A=VectorParameter() {group=Base_color, parameterName=Base_Color1, defaultValue=<rgba 1.0417 0 0.8032 1>}.RGB
   B=VectorParameter() {group=Base_color, parameterName=Base_Color2, defaultValue=<rgba 0 16.0283 21.3542 1>}.RGB
   Alpha=SphereMask(A=@n1.XYZ, B=Add(A=@n1.XYZ, B=@n2.RGB), Radius=ScalarParameter() {defaultValue=500, parameterName=SphereMaskRadius}) {hardnessPercent=50}
  )
  n4 = ScalarParameter() {defaultValue=1, parameterName=Emissive_Intensity}
  n5 = VectorParameter() {parameterName=MultTile, defaultValue=<rgba 1 1 0 1>}
  n6 = Panner(
   Coordinate=Multiply(A=TextureCoordinate() {uTiling=2, vTiling=0.25}, B=ComponentMask(None=@n5.RGB) {g=true, r=true})
   Speed=Multiply(A=Constant2Vector() {g=1, r=-0.5}, B=ScalarParameter() {defaultValue=1, parameterName=PannerSpeed})
  ) {speedX=-0.5, speedY=-1}
  n7 = VectorParameter() {parameterName=MultTile, defaultValue=<rgba 1 1 0 1>}
  n8 = Panner(
   Coordinate=Multiply(A=TextureCoordinate() {uTiling=2, vTiling=0.5}, B=ComponentMask(None=@n7.RGB) {g=true, r=true})
   Speed=Multiply(A=Constant2Vector() {g=-2, r=1}, B=ScalarParameter() {defaultValue=1, parameterName=PannerSpeed})
  ) {speedX=1, speedY=-2}
  n9 = VectorParameter() {parameterName=MultTile, defaultValue=<rgba 1 1 0 1>}
  n10 = Panner(
   Coordinate=Multiply(A=TextureCoordinate() {uTiling=2, vTiling=0.5}, B=ComponentMask(None=@n9.RGB) {g=true, r=true})
   Speed=Multiply(A=Constant2Vector() {g=1, r=-0.5}, B=ScalarParameter() {defaultValue=1, parameterName=PannerSpeed})
  ) {speedX=-0.5, speedY=-1}
  n11 = VectorParameter() {parameterName=MultTile, defaultValue=<rgba 1 1 0 1>}
  n12 = Panner(
   Coordinate=Multiply(A=TextureCoordinate() {uTiling=3}, B=ComponentMask(None=@n11.RGB) {g=true, r=true})
   Speed=Multiply(A=Constant2Vector() {g=-2, r=1}, B=ScalarParameter() {defaultValue=1, parameterName=PannerSpeed})
  ) {speedX=1, speedY=-2}
  n13 = LinearInterpolate(
   A=LinearInterpolate(
    A=TextureSample(UVs=@n6) {texture=&T_Inky_Smoke_Tile_Chromatic}.G
    B=TextureSample(UVs=@n8) {texture=&T_Inky_Smoke_Tile_Chromatic}.G
   )
   B=LinearInterpolate(
    A=TextureSample(UVs=@n10) {texture=&T_Inky_Smoke_Tile_Chromatic}.G
    B=TextureSample(UVs=@n12) {texture=&T_Inky_Smoke_Tile_Chromatic}.G
   )
  )
  n14 = Constant() {r=-0.5}
  MaterialFunctionCall(Emissive=Multiply(A=Multiply(A=@n3, B=@n4), B=Desaturation(None=@n13, Fraction=@n14)) {constB=1.5}) {materialFunction=&MF_FX_EmissiveAdjustments}
  n15 = ScalarParameter() {defaultValue=2, parameterName="Exponent (S)"}
  n16 = TextureCoordinate() {uTiling=0}
  n17 = DynamicParameter() {defaultValue=<rgba 0 0.5 1 1>, paramNames [PannerX,PannerY,Param3,Param4]}
  n18 = AppendVector(A=@n17.PannerX, B=@n17.PannerX)
  n19 = ScalarParameter() {defaultValue=0.5, parameterName=SoftMask_Radius}
  n20 = ScalarParameter() {defaultValue=-1, parameterName=SoftMask_Hardness}
  n21 = Multiply(A=Multiply(A=ParticleColor().A, B=SphereMask(A=@n16, B=@n18, Radius=@n19, Hardness=@n20)), B=@n13)
  n22 = Multiply(
   A=MaterialFunctionCall(Distance=ScalarParameter() {defaultValue=250, parameterName=Camera_FadeDistance}) {materialFunction=&Near_Camera_Fade}.Result
   B=DepthFade(Opacity=@n21, FadeDistance=ScalarParameter() {defaultValue=250, parameterName=Depth_FadeDistance})
  )
  Multiply(
   A=ScalarParameter() {defaultValue=0.5, parameterName=EventIntensity}
   B=Multiply(
    A=MaterialFunctionCall(Exponent=@n15) {materialFunction=&AxisAlignedFresnel}.Fresnel
    B=FeatureLevelSwitch(Default=@n22, ES2=@n21, ES3_1=@n21, SM4_REMOVED=@n22, SM5=@n22, SM6=@n22)
   )
  )
 ]
}
