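# Additive, unlit, two-sided VFX material for glowing animated smoke, named for the cracks in the MaskMaker takeover effect.
# A panned normal texture distorts scrolling detail noise, which feeds color gradients with hue shift and edge-masked emissive intensity.
# Opacity combines panned caustics, vertical gradients, a UV sphere mask, Fresnel terms, depth fade and camera fade.
# A separate panned distortion branch builds world position offset along vertex normals, and vertex color varies scroll timing.
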Material /BRMapCh6/FX/MaskMakerTakeOver/Materials/M_MaskMaker_Cracks_GlowSmoke_New_01
graph {
 properties {bAutomaticallySetUsageInEditor=false, blendMode=BLEND_Additive, shadingModel=MSM_Unlit, twoSided=true, usedShadingModels=Unlit}
 nodes [
  ComponentMask() {g=true, r=true}
  LinearInterpolate()
  n1 = VectorParameter() {group=Base_color, parameterName=Base_Color1, defaultValue=<rgba 1.0417 0 0.8032 1>}
  n2 = Panner(Coordinate=TextureCoordinate() {uTiling=0.1, vTiling=0.2}, Time=Time()) {speedY=0.2}
  n3 = VectorParameter() {group=Distortion, parameterName=Distortion UV Scale, sortPriority=3, defaultValue=<rgba 0.1 1 0 0>}
  n4 = VectorParameter() {group=Distortion, parameterName=Distortion UV Scale Edge, sortPriority=3, defaultValue=<rgba 0.1 1 0 0>}
  n5 = Multiply(
   A=LinearInterpolate(A=ComponentMask(None=@n3.RGB) {g=true, r=true}, B=ComponentMask(None=@n4.RGB) {g=true, r=true}, Alpha=NamedRerouteUsage())
   B=TextureCoordinate() {vTiling=0.5}
  )
  n6 = Multiply(A=Time(), B=NamedRerouteUsage())
  n7 = VertexColor()
  n8 = TextureSampleParameter2D(UVs=Panner(Coordinate=Multiply(A=@n5), Time=Add(A=@n6, B=@n7.R)) {speedY=1}) {parameterName=DistortTexture, samplerType=SAMPLERTYPE_Normal, texture=&T_Weap_Distort_N_02}
  n9 = TextureCoordinate()
  n10 = LinearInterpolate(A=ComponentMask(None=@n8.RGB) {g=true, r=true}, Alpha=ComponentMask(None=Multiply(A=@n9)) {g=true}) {constB=0}
  n11 = Constant() {r=3}
  n12 = VectorParameter() {group=Distortion, parameterName=Fire Micro UV Scale, sortPriority=3, defaultValue=<rgba 1 1 0 0>}
  n13 = VectorParameter() {group=Distortion, parameterName=Fire Micro UV Scale Edge, sortPriority=3, defaultValue=<rgba 1 1 0 0>}
  n14 = Multiply(
   A=TextureCoordinate() {vTiling=0.8}
   B=LinearInterpolate(A=ComponentMask(None=@n12.RGB) {g=true, r=true}, B=ComponentMask(None=@n13.RGB) {g=true, r=true}, Alpha=NamedRerouteUsage())
  )
  n15 = Multiply(A=Time(), B=NamedRerouteUsage())
  n16 = AppendVector(A=ScalarParameter() {parameterName=Panner_Speed_X}, B=ScalarParameter() {defaultValue=1, parameterName=Panner_Speed_Y})
  n17 = TextureSampleParameter2D(UVs=Add(A=Divide(A=@n10, B=@n11) {constB=20}, B=Panner(Coordinate=@n14, Time=@n15, Speed=@n16) {speedY=1.2})) {parameterName=DetailTexture, samplerType=SAMPLERTYPE_LinearColor, texture=&T_NoisePack_LinearColor}
  n18 = Multiply(
   A=ComponentMask(None=@n17.RGB) {g=true, r=true}
   B=ScalarParameter() {defaultValue=0.1, group=Base_color, parameterName=Base_texture_disturb}
  )
  n19 = MaterialFunctionCall(UV Channel=Multiply(A=TextureCoordinate() {vTiling=0.9}, B=ScalarParameter() {defaultValue=1, parameterName=BaseGradient_Offset})) {materialFunction=&LinearGradient}
  n20 = Multiply(A=OneMinus(None=@n19.VGradient), B=Power(Base=@n19.VGradient) {constExponent=4})
  n21 = Saturate(None=@n20)
  n22 = TextureSampleParameter2D(UVs=Add(A=Add(A=@n2, B=@n18), B=@n21)) {group=Base_color, parameterName=Base_color_lerp_mask, texture=&T_Inky_Smoke_Tile}
  LinearInterpolate(
   A=@n1.RGB
   B=VectorParameter() {group=Base_color, parameterName=Base_Color2, defaultValue=<rgba 0 16.0283 21.3542 1>}.RGB
   Alpha=Add(A=@n22.G, B=Multiply(A=@n21) {constB=5})
  )
  MaterialFunctionCall() {materialFunction=&FadeBasedOnViewAngle}
  Multiply(B=MaterialFunctionCall() {materialFunction=&LinearGradient}.VGradient)
  n23 = ScalarParameter() {defaultValue=1, parameterName=Emissive_Intensity}
  n24 = TextureCoordinate()
  n25 = Saturate(None=Power(
    Base=MaterialFunctionCall(UV Channel=@n24) {materialFunction=&LinearGradient}.VGradient
    Exp=ScalarParameter() {defaultValue=32, parameterName=EdgeMaskPower}
   ) {constExponent=32})
  n26 = ScalarParameter() {defaultValue=0.25, parameterName="Hue Shift Percentage (S)"}
  n27 = NamedRerouteUsage()
  n28 = Multiply(
   A=@n17.G
   B=LinearInterpolate(
    A=VectorParameter() {group=Base_color, parameterName=Base_Color_Top, defaultValue=<rgba 0.5026 0 1.0417 1>}.RGB
    B=@n1.RGB
    Alpha=Saturate(None=@n27)
   )
  )
  Multiply(
   A=LinearInterpolate(A=@n23, B=ScalarParameter() {defaultValue=1, parameterName=Emissive_Intensity_Edge}, Alpha=@n25)
   B=MaterialFunctionCall(Hue Shift Percentage=@n26, Texture=@n28) {materialFunction=&HueShift}.Result
  )
  n29 = Constant() {r=0.3}
  n30 = TextureSampleParameter2D(UVs=Panner(Coordinate=TextureCoordinate() {uTiling=0.1, vTiling=0.1}) {speedY=0.1}) {parameterName=MainTexture, texture=&T_ShieldCaustics}
  n31 = LinearInterpolate(Alpha=LinearInterpolate(Alpha=Power(Base=@n30.G) {constExponent=0.6}) {constB=3}) {constA=-0.3}
  n32 = TextureCoordinate() {uTiling=0.2}
  n33 = Multiply(A=OneMinus(None=ComponentMask(None=@n32) {g=true})) {constB=3}
  n34 = Multiply(A=Multiply(A=Saturate(None=@n31), B=Saturate(None=@n33))) {constB=0.3}
  n35 = ComponentMask(None=Multiply(
    A=Add(A=@n29, B=@n34)
    B=MaterialFunctionCall() {materialFunction=&LinearGradient}.VGradient
   ) {constA=1}) {r=true}
  n36 = Multiply(A=@n35) {constB=3}
  n37 = Saturate(None=LinearInterpolate(A=@n35, B=LinearInterpolate(Alpha=@n36)))
  n38 = SphereMask(A=TextureCoordinate() {coordinateIndex=1}, B=Constant2Vector() {g=0.5, r=0.5}) {attenuationRadius=0.5, hardnessPercent=0}
  n39 = Multiply(A=Add(A=NamedRerouteUsage(), B=ScalarParameter() {defaultValue=1, group=Opacity, parameterName=Global_Opacity}), B=@n38)
  n40 = Multiply(
   A=MaterialFunctionCall(Exponent=ScalarParameter() {defaultValue=8, parameterName="FadeSensitivity (optional) (S)"}) {materialFunction=&AxisAlignedFresnel}.Fresnel
   B=Multiply(A=Multiply(A=@n37, B=@n39), B=@n17.G)
  )
  n41 = DepthFade(FadeDistance=ScalarParameter() {defaultValue=25, parameterName=DepthFade_Distance}) {fadeDistanceDefault=25}
  n42 = ScalarParameter() {defaultValue=2, parameterName=ExponentIn}
  Multiply(
   A=Multiply(A=Multiply(A=@n40, B=@n41), B=Fresnel(ExponentIn=@n42))
   B=MaterialFunctionCall(
    Fade Length=ScalarParameter() {defaultValue=512, parameterName="Fade Length (S)"}
    Fade Offset=ScalarParameter() {defaultValue=256, parameterName="Fade Offset (S)"}
   ) {materialFunction=&CameraDepthFade}.Result
  )
  n43 = CurveAtlasRowParameter(CurveTime=ComponentMask(None=TextureCoordinate()) {g=true}) {
   atlas=/Game/Athena/LinearColorCurves/VFX_CurveAtlas
   curve=/Game/Accessories/FORT_Backpacks/Backpack_F_MED_Tower_Sentinel/FX/Curve_Tower_Sentinel_Female
   defaultValue=15
   parameterName=Global_color
  }
  Multiply(A=@n28, B=@n43)
  Multiply(A=@n26, B=Multiply(A=ScalarParameter() {defaultValue=1, parameterName=CenterFadeSaturation}, B=NamedRerouteUsage()))
  n44 = WorldPosition()
  n45 = VectorParameter() {parameterName=SphereMaskPosition, defaultValue=<rgba 0 0 0 1>}
  Multiply(
   A=SphereMask(A=@n44.XYZ, B=Add(A=@n44.XYZ, B=@n45.RGB), Radius=ScalarParameter() {defaultValue=500, parameterName=SphereMaskRadius}) {hardnessPercent=50}
   B=@n23
  )
  NamedRerouteDeclaration(None=@n25) {name=EdgeMask, nodeColor=<rgba 1 0.0234 0 1>}
  NamedRerouteDeclaration(None=Multiply(A=@n20, B=ScalarParameter() {defaultValue=8, parameterName=BaseGradient_Intensity})) {name=BaseGradient, nodeColor=<rgba 0.8984 1 0 1>}
  NamedRerouteDeclaration(None=@n38) {name=SphereMaskCenter, nodeColor=<rgba 0.6172 0 1 1>}
  n46 = VertexNormalWS()
  n47 = ScalarParameter() {defaultValue=1, parameterName=WPO_TileX}
  n48 = ScalarParameter() {defaultValue=0.1, parameterName=WPO_TileY}
  n49 = Panner(
   Coordinate=Multiply(A=AppendVector(A=@n47, B=@n48), B=TextureCoordinate())
   Time=Multiply(A=Time(), B=ScalarParameter() {defaultValue=1, group=WPO, parameterName=WPO_MultSpeed, sortPriority=0})
   Speed=AppendVector(
    A=ScalarParameter() {group=WPO, parameterName=WPO_Speed_X, sortPriority=0}
    B=ScalarParameter() {defaultValue=1, group=WPO, parameterName=WPO_Speed_Y, sortPriority=0}
   )
  ) {bFractionalPart=true, speedY=0.8}
  n50 = ComponentMask(None=TextureSample(UVs=@n49) {samplerType=SAMPLERTYPE_Normal, texture=&T_Weap_Distort_N_02}.RGB) {g=true, r=true}
  n51 = TextureSample(UVs=@n49) {texture=&T_Smooth_tile}
  n52 = ComponentMask(None=TextureCoordinate()) {g=true}
  n53 = Multiply(A=LinearInterpolate(B=AppendVector(A=@n50, B=@n51.G), Alpha=OneMinus(None=@n52)), B=ScalarParameter() {parameterName=WPO_Intensity}) {constB=10}
  NamedRerouteDeclaration(None=LinearInterpolate(
    A=Multiply(A=@n46, B=@n53)
    B=@n46
    Alpha=MaterialFunctionCall() {materialFunction=&LinearGradient}.VGradient
   )) {name=WPO, nodeColor=<rgba 0.7812 1 0 1>}
  n54 = ScalarParameter() {defaultValue=0.05, parameterName=SpeedClampMin}
  n55 = ScalarParameter() {defaultValue=0.1, parameterName=SpeedClampMax}
  n56 = VertexColor()
  NamedRerouteDeclaration(None=Add(
    A=ScalarParameter() {defaultValue=1, group=Panner Speed, parameterName=Speed}
    B=LinearInterpolate(A=@n54, B=@n55, Alpha=@n56.R) {constA=0.05, constB=0.1}
   )) {name=Speed, nodeColor=<rgba 0.5938 0 1 1>}
  NamedRerouteUsage()
 ]
}
