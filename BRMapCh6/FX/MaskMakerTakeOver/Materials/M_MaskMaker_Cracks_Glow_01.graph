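# Additive, unlit, two-sided material for the glowing cracks of the MaskMaker takeover effect, with EventIntensity scaling brightness and pan speed.
# Opacity combines panning caustics and normal-distorted noise with UV gradients, a center mask, Fresnel and depth, camera and view-angle fades.
# Color blends two tints by a smoke mask, then applies a curve-atlas gradient and center-weighted hue shift.
# A scrolling noise offset along vertex normals drives WPO, and vertex red varies panning.
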Material /BRMapCh6/FX/MaskMakerTakeOver/Materials/M_MaskMaker_Cracks_Glow_01
graph {
 properties {bAutomaticallySetUsageInEditor=false, blendMode=BLEND_Additive, shadingModel=MSM_Unlit, twoSided=true, usedShadingModels=Unlit}
 nodes [
  n1 = MaterialFunctionCall("FadeSensitivity (optional)"=ScalarParameter() {defaultValue=8, parameterName="FadeSensitivity (optional) (S)"}) {materialFunction=&FadeBasedOnViewAngle}
  n2 = Constant() {r=0.3}
  n3 = TextureSampleParameter2D(UVs=Panner(Coordinate=TextureCoordinate() {uTiling=0.1, vTiling=0.1}) {speedY=0.1}) {parameterName=MainTexture, texture=&T_ShieldCaustics}
  n4 = LinearInterpolate(Alpha=LinearInterpolate(Alpha=Power(Base=@n3.G) {constExponent=0.6}) {constB=3}) {constA=-0.3}
  n5 = TextureCoordinate() {uTiling=0.2}
  n6 = Multiply(A=OneMinus(None=ComponentMask(None=@n5) {g=true})) {constB=3}
  n7 = Multiply(A=Multiply(A=Saturate(None=@n4), B=Saturate(None=@n6))) {constB=0.3}
  n8 = ComponentMask(None=Multiply(
    A=Add(A=@n2, B=@n7)
    B=MaterialFunctionCall() {materialFunction=&LinearGradient}.VGradient
   ) {constA=1}) {r=true}
  n9 = LinearInterpolate(A=@n8, B=LinearInterpolate(Alpha=Multiply(A=@n8) {constB=3}))
  n10 = ScalarParameter() {defaultValue=1, group=Opacity, parameterName=Global_Opacity}
  n11 = SphereMask(A=TextureCoordinate() {coordinateIndex=1}, B=Constant2Vector() {g=0.5, r=0.5}) {attenuationRadius=0.5, hardnessPercent=0}
  n12 = VectorParameter() {group=Distortion, parameterName=Distortion UV Scale, sortPriority=3, defaultValue=<rgba 0.1 1 0 0>}
  n13 = Multiply(A=ComponentMask(None=@n12.RGB) {g=true, r=true}, B=TextureCoordinate() {vTiling=0.5})
  n14 = Multiply(A=Time(), B=NamedRerouteUsage())
  n15 = VertexColor()
  n16 = TextureSampleParameter2D(UVs=Panner(Coordinate=Multiply(A=@n13), Time=Add(A=@n14, B=@n15.R)) {speedY=1}) {parameterName=DistortTexture, samplerType=SAMPLERTYPE_Normal, texture=&T_Weap_Distort_N_02}
  n17 = TextureCoordinate()
  n18 = LinearInterpolate(A=ComponentMask(None=@n16.RGB) {g=true, r=true}, Alpha=ComponentMask(None=Multiply(A=@n17)) {g=true}) {constB=0}
  n19 = Constant() {r=3}
  n20 = Multiply(
   A=TextureCoordinate() {vTiling=0.8}
   B=ComponentMask(None=VectorParameter() {group=Distortion, parameterName=Fire Micro UV Scale, sortPriority=3, defaultValue=<rgba 1 1 0 0>}.RGB) {g=true, r=true}
  )
  n21 = Multiply(A=Time(), B=NamedRerouteUsage())
  n22 = TextureSampleParameter2D(UVs=Add(A=Divide(A=@n18, B=@n19) {constB=20}, B=Panner(Coordinate=@n20, Time=@n21) {speedY=1.2})) {parameterName=DetailTexture, samplerType=SAMPLERTYPE_LinearColor, texture=&T_NoisePack_LinearColor}
  n23 = Multiply(A=Multiply(A=Saturate(None=@n9), B=Multiply(A=@n10, B=@n11)), B=@n22.G)
  n24 = ScalarParameter() {defaultValue=25, parameterName=DepthFade_Distance}
  n25 = Multiply(
   A=Multiply(A=Multiply(A=@n1.Result, B=@n23), B=DepthFade(FadeDistance=@n24) {fadeDistanceDefault=25})
   B=Fresnel(ExponentIn=ScalarParameter() {defaultValue=2, parameterName=ExponentIn})
  )
  n26 = MaterialFunctionCall(
   Fade Length=ScalarParameter() {defaultValue=512, parameterName="Fade Length (S)"}
   Fade Offset=ScalarParameter() {defaultValue=256, parameterName="Fade Offset (S)"}
  ) {materialFunction=&CameraDepthFade}
  n27 = MaterialFunctionCall(UV Channel=Add(A=TextureCoordinate() {uTiling=0.12}, B=Constant2Vector() {r=-0.04})) {materialFunction=&LinearGradient}
  n28 = Multiply(A=Multiply(A=@n27.UGradient, B=OneMinus(None=@n27.UGradient))) {constB=8}
  Multiply(A=ScalarParameter() {defaultValue=1, parameterName=EventIntensity}, B=Multiply(A=Multiply(A=@n25, B=@n26.Result), B=Saturate(None=@n28)))
  n29 = WorldPosition()
  n30 = Add(A=@n29.XYZ, B=VectorParameter() {parameterName=SphereMaskPosition, defaultValue=<rgba 0 0 0 1>}.RGB)
  n31 = ScalarParameter() {defaultValue=500, parameterName=SphereMaskRadius}
  n32 = ScalarParameter() {defaultValue=0.25, parameterName="Hue Shift Percentage (S)"}
  n33 = Multiply(A=ScalarParameter() {defaultValue=1, parameterName=CenterFadeSaturation}, B=NamedRerouteUsage())
  n34 = Panner(Coordinate=TextureCoordinate() {uTiling=0.1, vTiling=0.2}, Time=Time()) {speedY=0.2}
  n35 = Multiply(
   A=ComponentMask(None=@n22.RGB) {g=true, r=true}
   B=ScalarParameter() {defaultValue=0.1, group=Base_color, parameterName=Base_texture_disturb}
  )
  n36 = Multiply(A=MaterialFunctionCall(UV Channel=TextureCoordinate() {vTiling=1.1}) {materialFunction=&LinearGradient}.VGradient)
  n37 = Saturate(None=Multiply(A=OneMinus(None=@n36), B=Power(Base=@n36) {constExponent=4}))
  n38 = TextureSampleParameter2D(UVs=Add(A=Add(A=@n34, B=@n35), B=@n37)) {group=Base_color, parameterName=Base_color_lerp_mask, texture=&T_Inky_Smoke_Tile}
  n39 = Multiply(A=@n37) {constB=5}
  n40 = Multiply(
   A=@n22.G
   B=LinearInterpolate(
    A=VectorParameter() {group=Base_color, parameterName=Base_Color1, defaultValue=<rgba 1.0417 0 0.8032 1>}.RGB
    B=VectorParameter() {group=Base_color, parameterName=Base_Color2, defaultValue=<rgba 0 16.0283 21.3542 1>}.RGB
    Alpha=Add(A=@n38.G, B=@n39)
   )
  )
  n41 = CurveAtlasRowParameter(CurveTime=ComponentMask(None=TextureCoordinate()) {g=true}) {
   atlas=/Game/Athena/LinearColorCurves/VFX_CurveAtlas
   curve=/Game/Accessories/FORT_Backpacks/Backpack_F_MED_Tower_Sentinel/FX/Curve_Tower_Sentinel_Female
   defaultValue=15
   parameterName=Global_color
  }
  Multiply(
   A=Multiply(A=SphereMask(A=@n29.XYZ, B=@n30, Radius=@n31) {hardnessPercent=50}, B=ScalarParameter() {defaultValue=1, parameterName=Emissive_Intensity})
   B=MaterialFunctionCall(Hue Shift Percentage=Multiply(A=@n32, B=@n33), Texture=Multiply(A=@n40, B=@n41)) {materialFunction=&HueShift}.Result
  )
  NamedRerouteDeclaration(None=@n11) {name=SphereMaskCenter, nodeColor=<rgba 0.6172 0 1 1>}
  n42 = ScalarParameter() {defaultValue=1, group=Panner Speed, parameterName=Speed}
  n43 = LinearInterpolate(Alpha=VertexColor().R) {constA=0.05, constB=0.1}
  NamedRerouteDeclaration(None=Multiply(A=ScalarParameter() {defaultValue=1, parameterName=EventIntensity}, B=Add(A=@n42, B=@n43))) {name=Speed, nodeColor=<rgba 0.5938 0 1 1>}
  n44 = VertexNormalWS()
  n45 = Panner(
   Coordinate=TextureCoordinate() {vTiling=0.1}
   Time=Multiply(A=Time(), B=ScalarParameter() {defaultValue=1, group=WPO, parameterName=WPO_Speed, sortPriority=0})
  ) {bFractionalPart=true, speedY=0.8}
  n46 = ComponentMask(None=TextureSample(UVs=@n45) {samplerType=SAMPLERTYPE_Normal, texture=&T_Weap_Distort_N_02}.RGB) {g=true, r=true}
  n47 = TextureSample(UVs=@n45) {texture=&T_Smooth_tile}
  n48 = ComponentMask(None=TextureCoordinate()) {g=true}
  n49 = Multiply(A=LinearInterpolate(B=AppendVector(A=@n46, B=@n47.G), Alpha=OneMinus(None=@n48)), B=ScalarParameter() {parameterName=WPO_Intensity}) {constB=10}
  NamedRerouteDeclaration(None=LinearInterpolate(
    A=Multiply(A=@n44, B=@n49)
    B=@n44
    Alpha=MaterialFunctionCall() {materialFunction=&LinearGradient}.VGradient
   )) {name=WPO, nodeColor=<rgba 0.7812 1 0 1>}
  NamedRerouteUsage()
 ]
}
